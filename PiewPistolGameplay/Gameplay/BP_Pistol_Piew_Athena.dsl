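; BP_Pistol_Piew_Athena is the Piew Pistol weapon, an overheating pistol with a charged super shot fired while aiming down sights.
; ADS and overheat changes send gameplay events and adjust audio, while timer-driven charge functions toggle super shot readiness and charge feedback.
; The Supershot tag swaps in an alternate lightweight projectile and muzzle flash.
; At zero ammo, the server triggers a delayed throwaway ability, then removes the item after a fallback delay.
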(event FortWeapon|EventOnUnEquippedAndDetached
  ;[0.15 0.15 0.15 0.5 50] OnUnEquipped
  (|CallParentFunction)
  (bind _chargeupaudio (Variables|Sound|GetChargeUpAudio)
    (:then
      (Audio|Components|Audio|Stop _chargeupaudio))
    (:else)))

(event EventBeginPlay
  ;[0 0.386 1 0.5 50] BeginPlay
  (|CallParentFunction)
  (bind _asplayer_pawn_athena (Utilities|Casting|CastToPlayerPawn_Athena (Game|GetInstigator)))
  (Variables|Default|SetPlayer _asplayer_pawn_athena)
  (Overheating|AssignOnOverheatValueChanged (EventDispatchers|CreateEvent :!event Custom|OnOverheatValueChanged_Event)))

(event Custom|OnOverheatValueChanged_Event
  (params (CurrentValue :type Float%20%28single-precision%29) (MaxValue :type Float%20%28single-precision%29) (State :type EFortWeaponOverheatState%20Enum) (bIsInCooldown :type Boolean))
  ;[1 0.359 0.003259 0.5 50] Set Server Overheat and Overheated Value
  (Variables|Default|SetOverheated bIsInCooldown)
  (if (Networking|IsServer)
    (Variables|Default|SetServerOverheat CurrentValue)))

(event Custom|AmmoCountChanged_Feedback
  (params (Quantity :type Integer))
  ;[0.01482 0.15 0.06742 0.5] Check for Ammo Count and Activate Throw Away Animation
  (Utilities|ClientRole
    (:Server
      (if (== (select (FortWeapon|Ammo|HasInfiniteUses) 1 (FortWeapon|GetRemainingAmmo)) 0)
        (Utilities|Time|SetTimerbyEvent (delegate Custom|PlayThrowawayAnim) (ScalableFloat|GetValueatLevel (Variables|Default|GetToss%20Delay))) ; Delay to give the FX some time to start the action
      ))))

(event FortWeapon|EventOnEquippedAndAttached
  ;[0.1426 0.0775 0.7716 0.5 50] Charge / Supershots
  (Default|BindEventtoOnAimDownSightsChanged (delegate Custom|onAimDownSightsChanged_Event))
  (EventDispatchers|BindEventtoOnConsumeAmmo (EventDispatchers|CreateEvent :!event Custom|AmmoCountChanged_Feedback)))

(event Custom|onAimDownSightsChanged_Event
  (params (AimDownSights :type Boolean))
  (bind _overheated (Variables|Default|GetOverheated))
  (bind _returnvalue (not _overheated))
  ;[0.1426 0.0775 0.7716 0.5 50] Charge / Supershots
  (bind _output_get (Variables|Default|SetAim%20Down%20Sights AimDownSights))
  (Utilities|FlowControl|Sequence
    (:then_0
      (CallFunction|UpdateOverheatAudio (and _output_get _returnvalue))
      (if (and _output_get _returnvalue)
        (if (not (Variables|Default|GetCanFireSuperShot))
          (CallFunction|SendGameplayEvent "(TagName=\"Event.PiewPistol.ADS\")"))
        (else
          (CallFunction|SendGameplayEvent "(TagName=\"Event.PiewPistol.Hipfire\")"))))
    (:then_1
      (if (not (Variables|Default|GetAim%20Down%20Sights))
        (CallFunction|PlayADSOutAudio)))))

(event Custom|RemoveShot
  ;[0.1426 0.0775 0.7716 0.5 50] Charge / Supershots
  (CallFunction|SendGameplayEvent "(TagName=\"Event.PiewPistol.Hipfire\")"))

(event Custom|PlayerCanSuperShot
  ;[0.1426 0.0775 0.7716 0.5 50] Charge / Supershots
  (Variables|Default|SetCanFireSuperShot true))

(event Custom|PlayThrowawayAnim
  ;[0.01482 0.15 0.06742 0.5] Check for Ammo Count and Activate Throw Away Animation
  (bind _asfort_player_pawn (Utilities|Casting|CastToFortPlayerPawn (Game|GetInstigator)))
  (Ability|SendGameplayEventtoActor _asfort_player_pawn "(TagName=\"PiewPistol.Item.AmmoChanged\")") ; Send event to trigger throw away GA
  (Utilities|Time|SetTimerbyEvent (delegate Custom|FallbackRemoveWeapon) (ScalableFloat|GetValueatLevel (Variables|Default|GetFallback%20Delay))))

(event Custom|PlayerCannotSuperShot
  ;[0.1426 0.0775 0.7716 0.5 50] Charge / Supershots
  (Variables|Default|SetCanFireSuperShot false))

(event Custom|Charge%20Super%20Shot
  ;[0.1426 0.0775 0.7716 0.5 50] Charge / Supershots
  (CallFunction|PlayerCannotSuperShot)
  (bind _returnvalue (Utilities|Time|SetTimerbyEvent (delegate Custom|FinishCharging) (Variables|Default|GetCharge%20Up%20Time)))
  (Variables|Default|SetCharging%20Timer _returnvalue)
  (Variables|Sound|GetChargeUpAudio
    (:then)
    (:else
      (bind _returnvalue_2 (Audio|SpawnSoundAttached (Variables|Sound|GetTargetingAudio) (Variables|Visual|GetWeaponMesh)))
      (Variables|Sound|SetChargeUpAudio _returnvalue_2))))

(event Custom|FallbackRemoveWeapon
  ;[0.01482 0.15 0.06742 0.5] Check for Ammo Count and Activate Throw Away Animation
  (Inventory|RemoveItemFromPlayerByGuid (Game|GetInstigatorController) (FortWeapon|GetInventoryGUID) -1 true))

(event Custom|FinishCharging
  ;[0.1426 0.0775 0.7716 0.5 50] Charge / Supershots
  (CallFunction|PlayerCanSuperShot)
  (Variables|Default|SetCanFireSuperShot true)
  (Components|Activation|Activate (Variables|FX|GetChargeUp_Placeholder) false))

(event Custom|Remove%20SuperShot
  ;[0.1426 0.0775 0.7716 0.5 50] Charge / Supershots
  ;;[1 0 0 0.5 50] Remove Super Shot
  (CallFunction|PlayerCannotSuperShot)
  (Variables|Default|SetCanFireSuperShot false)
  (Utilities|Time|ClearandInvalidateTimerbyHandle (Variables|Default|GetCharging%20Timer))
  (Components|Activation|Deactivate (Variables|FX|GetChargeUp_Placeholder))
  (bind _chargeupaudio (Variables|Sound|GetChargeUpAudio)
    (:then
      (Audio|Components|Audio|Stop _chargeupaudio))
    (:else)))

(event Custom|Play%20Reload%20FX
  ;[0.7005 0.1623 0.9293 0.5 50] Reload FX
  (Effects|SpawnFXSystemAttached (Variables|FX|GetReloadOverHeat_FX) (Variables|Visual|GetWeaponMesh) "attach_mag"))

(event Custom|Add%20Building%20Up%20FX
  ;[0.7005 0.1623 0.9293 0.5 50] Charge FX
  (Niagara|SpawnSystemAttached "/PiewPistolGameplay/Assets/FX/NS_PiewPistol_ADS_ChargingUp.NS_PiewPistol_ADS_ChargingUp" (Variables|FX|GetChargeUp_Placeholder) "None" "0, 0, 0" "0, 0, 0" "SnapToTarget" true true "ManualRelease"))

(event EventMuzzleFlashFX
  (params (Persistent%20Fire :type Boolean))
  ;[0.7005 0.1623 0.9293 0.5 50] MuzzleFlash FX
  (|CallParentFunction Persistent%20Fire)
  (if (GameplayTags|HasMatchingGameplayTag (Variables|Default|GetPlayer) "(TagName=\"PiewPistol.Supershot\")")
    (Niagara|SpawnSystemAttached "/PiewPistolGameplay/Assets/FX/NS_MF_PiewPistol_ADS.NS_MF_PiewPistol_ADS" (Variables|Visual|GetWeaponMesh) "Muzzle")))

(event Overheating|EventOnStartOverheated
  ;[0.5885 0.2836 0.01321 0.5 50] Used to trigger RelaxedPose when Overheated
  (Variables|Default|SetOverheated true)
  (CallFunction|SendGameplayEvent "(TagName=\"Event.PiewPistol.Hipfire\")")
  (CallFunction|SendGameplayEvent "(TagName=\"Event.PiewPistol.Reload\")")
  (Effects|Components|ParticleSystem|SetBoolParameter (Variables|FX|GetChargeUp_Placeholder) "bCanSuperShot" (not (Variables|Default|GetOverheated))))

(event Overheating|EventOnFinishedCooling
  ;[0.5885 0.2836 0.01321 0.5 50] Used to trigger RelaxedPose when Overheated
  (Variables|Default|SetOverheated false)
  (if (and (not (Variables|Default|GetCanFireSuperShot)) (FortWeapon|IsTargeting))
    (CallFunction|SendGameplayEvent "(TagName=\"Event.PiewPistol.ADS\")")
    (else
      _))
  (Effects|Components|ParticleSystem|SetBoolParameter (Variables|FX|GetChargeUp_Placeholder) "bCanSuperShot" (not (Variables|Default|GetOverheated))))

(event Overheating|EventOnLocalFailedToFireWhenOverheated
  ;[0.0208 1 0.3741 0.5 50] Overheat audio logic for fire mode changes
  (Audio|PlaySound2D "/Game/Athena/Sounds/Items/Ability_CantUse_Cue.Ability_CantUse_Cue"))

(event Custom|UpdateOverheatAudio
  (params (bIsCharged :type Boolean))
  (bind _weaponoverheatsoundcomponent (Variables|Sound|GetWeaponOverheatSoundComponent))
  ;[0.0208 1 0.3741 0.5 50] Overheat audio logic for fire mode changes
  (Class|WeaponOverheatSoundComponent|UpdateOverheatCurves _weaponoverheatsoundcomponent (select bIsCharged (Variables|Sound|GetChargedFireOverheatVolumeCurve) (Variables|Sound|GetNormalFireOverheatVolumeCurve)) (select bIsCharged (Variables|Sound|GetChargedFireOverheatPitchCurve) (Variables|Sound|GetNormalFireOverheatPitchCurve)))
  (Class|WeaponOverheatSoundComponent|ClearSweetenerAudioComponent _weaponoverheatsoundcomponent))

(event Custom|PlayADSOutAudio
  (bind _returnvalue (SoundLibrary|GetSoundLibraryComponent "(TagName=\"SoundLibrary.Weapon\")"))
  ;[0.0208 1 0.3741 0.5 50] Overheat audio logic for fire mode changes
  (bind _player (Variables|Default|GetPlayer)
    (:then
      (if (Pawn|IsLocallyViewed _player)
        (Variables|Default|GetOverheated
          (:then)
          (:else
            (Utilities|IsValid _returnvalue
              (:Is%20Valid
                (SoundLibrary|PlayEvent _returnvalue "(TagName=\"SoundLibrary.PiewPistol.ADS.Out\")")))))))
    (:else)))

(fn GetLightweightProjectile
  (if (GameplayTags|HasMatchingGameplayTag (Ability|GetFortAbilitySystemComponent (Variables|Default|GetPlayer)) "(TagName=\"PiewPistol.Supershot\")")
    (return (Variables|Default|GetAltLWP))
    (else
      (bind _returnvalue (|CallParentFunction))
      (return _returnvalue))))

(fn OnRep_ServerOverheat
  (Default|CallOnServerOverheatChanged))

(fn SendGameplayEvent
  (params (Event%20Tag :type Gameplay%20Tag%20Structure))
  (Ability|SendGameplayEventtoActor (Variables|Default|GetPlayer) Event%20Tag (Utilities|Struct|MakeGameplayEventData Event%20Tag (Variables|Default|GetPlayer))))

(fn ShouldValidateLightWeightProjectile
  (return (Variables|Default|GetValidateLWP?)))

(fn ConstructionScript
  (|CallParentFunction))
