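; Gameplay ability for the Piew pistol's primary action, holding only two function overrides with constant results.
; CanApplyHolster always returns false, whatever holster ID is passed in.
; CanActivateAbility always returns true without reading the actor info or spec handle.
; No firing or activation logic appears here.
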(fn CanApplyHolster
  (params (HolsterId :type Name%20%28by%20ref%29 :default "None"))
  (return false))

(fn CanActivateAbility
  (params (ActorInfo :type Gameplay%20Ability%20Actor%20Info%20Structure) (Handle :type Gameplay%20Ability%20Spec%20Handle%20Structure))
  (return 0 true))
