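; Targeting ability for the Piew Pistol that grants a SuperShot gameplay effect while the ability input is held.
; On activation it commits, casts the current weapon to the Piew pistol, and sets targeting on.
; If the commit succeeded and the weapon is targeting, it applies the effect and stores its handle; otherwise it ends.
; Input release ends the ability, which removes the effect by handle; debug prints mark activation and deactivation.
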(event Ability|EventActivateAbility
  (bind _returnvalue (CallFunction|GetAvatarasFortPlayerPawn))
  (bind _returnvalue_2 (Weapon|GetCurrentWeapon _returnvalue))
  ; Activate
  (bind _returnvalue_3 (Ability|CommitAbility))
  (bind _asbp_pistol_piew_athena (Utilities|Casting|CastToBP_Pistol_Piew_Athena _returnvalue_2))
  (FortWeapon|SetTargeting _asbp_pistol_piew_athena true)
  (if (and _returnvalue_3 (FortWeapon|IsTargeting _returnvalue_2))
    (bind _returnvalue_4 (Ability|ApplyGameplayEffectToOwner "/PiewPistolGameplay/Gameplay/GE_PiewPistol_SuperShot.GE_PiewPistol_SuperShot_C"))
    (Variables|Default|SetEffectHandle _returnvalue_4)
    (Development|PrintString "Activated" true true "(R=0.003422,G=1.000000,B=0.000000,A=1.000000)" 5.0)
    (else
      (Ability|EndAbility))))

(event Ability|EventOnAbilityInputReleased
  ; Activate
  (Ability|EndAbility))

(event Ability|EventOnEndAbility
  (params (bWasCancelled :type Boolean))
  ; End Ability
  (|CallParentFunction bWasCancelled)
  (Ability|RemoveGameplayEffectFromOwnerWithHandle (Variables|Default|GetEffectHandle))
  (Development|PrintString "Deactivated" true true "(R=1.000000,G=0.000000,B=0.112739,A=1.000000)" 5.0))
