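; Custom application requirement for Piew Pistol hit gameplay effects, permitting them only when the weapon is not targeting (aiming down sights).
; It takes the effect causer from the spec's effect context, casts it to a Fort weapon, and returns the inverse of its IsTargeting state.
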(fn CanApplyGameplayEffect
  (params (GameplayEffect :type Gameplay%20Effect%20Object%20Reference) (Spec :type Gameplay%20Effect%20Spec%20Structure%20%28by%20ref%29) (ASC :type Ability%20System%20Component%20Object%20Reference))
  (bind _asfort_weapon (Utilities|Casting|CastToFortWeapon (Ability|EffectContext|GetEffectCauser (Ability|GameplayEffect|GetEffectContextfromSpec Spec))))
  (return (not (FortWeapon|IsTargeting _asfort_weapon))))
